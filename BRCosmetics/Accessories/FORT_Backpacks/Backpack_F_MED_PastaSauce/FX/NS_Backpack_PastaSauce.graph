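# Niagara effect for the PastaSauce backpack cosmetic that adds a pulsing purple glow sprite and light.
# One emitter burst-spawns a particle that outlives its lifetime and feeds both a light renderer and a fixed-facing sprite renderer.
# A waveform modulates its color, from which the light color is derived with time-of-day scaling.
# A second color scale then adjusts brightness by time of day and lowers alpha in the front end.
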NiagaraSystem /BRCosmetics/Accessories/FORT_Backpacks/Backpack_F_MED_PastaSauce/FX/NS_Backpack_PastaSauce
dependencies {usedDataInterfaces [], usedDynamicInputs [], usedModules [&SystemState]}
emitters {Light_And_Glow {
  emitterUpdateScript [
   &EmitterState(
    Inactive Response=::NewEnumerator1
    Life Cycle Mode=::NewEnumerator1
    Loop Behavior=::NewEnumerator1
    Loop Duration Mode=::NewEnumerator1
    UseLoopDelay=false
   )
   &SpawnBurst_Instantaneous(Use Spawn Probability=false, UseLoopCountLimit=false)
  ]
  particleSpawnScript [
   &InitializeParticle(
    Color=<rgba 0.4319 0 1 1.333>
    Lifetime=5
    Position Offset=<0 0.1 0>
    Sprite Size Mode=::NewEnumerator3
    Uniform Sprite Size=30
    UsePositionOffset=false
    Write Material Random=true
   )
   &SystemLocation(MaskSpawnGroup=false, Offset=<0 -6 0>)
  ]
  particleUpdateScript [
   &ParticleState(Kill Particles When Lifetime Has Elapsed=false)
   &ScaleColor(
    Scale RGB=&VectorFromFloat(Value=&Waveform(
      Waveform Count=::NewEnumerator1
      "Waveform [1]"=::NewEnumerator0
      "[1] Frequency"=0.618
      "[1] Amplitude Min/Max"=<0.25 0.5>
      "[2] Average"=false
      "[2] Input Position"=0
      "[2] Phase"=$Emitter.Age
      "[2] Phase Scale"=1
      "[2] Compound Frequency"=2
      "[2] Frequency"=1
      "[2] Amplitude Min/Max"=<0.25 0.5>
      "[2] Scale"=1
      "[2] Bias"=0
      "[2] Exponent"=1
      "[2] Pulse Duration"=0.1
      "[2] Clip Triangle Min/Max"=<0 0>
      "[2] Random Seed"=0
      "Waveform [3] Blend"=::NewEnumerator5
     ))
    ScaleA=false
    ScaleRGB=true
   )
   /SetVariables(
    Particles.LightColor=&MultiplyLinearColor(
     Linear Color=$Particles.Color
     Scale Factor=&MakeFloat4FromVectorAndFloat(
      "Vector (RGB)"=&VectorFromFloat(Value=&Multiply_Float(A=10, B=&FNTimeOfDayScalar(Night Scale=1, Morning Scale=1.5, Evening Scale=1.5)))
      "Float (Alpha)"=1
     )
    )
    Particles.LightPosition:NiagaraPosition=&AddVectorToPosition(Position=$Engine.Emitter.SimulationPosition, Vector=<0 -12.5 0>)
   ) {isSetParameters=true}
   &ScaleColor(
    Scale Alpha=&FN_FrontEndFloatModifier(Gameplay=1, Front End=0.5)
    Scale RGB=&VectorFromFloat(Value=&Multiply_Float(A=&FNTimeOfDayScalar(Night Scale=1.2, Morning Scale=1, Day Scale=1.33, Evening Scale=1), B=1))
    ScaleA=true
    ScaleRGB=true
   ) {module=ScaleColor001}
   &SpriteFacingAndAlignment(Sprite Facing=<0 -1 0>, WriteAlignment=false, WriteFacing=true)
  ]
  rendererClasses [&NiagaraLightRendererProperties, &NiagaraSpriteRendererProperties]
  renderers [
   {rendererClass=&NiagaraLightRendererProperties, rendererIndex=0}
   {rendererClass=&NiagaraSpriteRendererProperties, rendererIndex=1}
  ]
 }}
summary {
 systemName=NS_Backpack_PastaSauce
 emitters [{
   emitterName=Light_And_Glow
   rendererClasses [&NiagaraLightRendererProperties, &NiagaraSpriteRendererProperties]
  }]
 userVariables [
  User.bIsFrontEnd {type=NiagaraBool}
  User.bIsFrontEndPreview {type=NiagaraBool}
  User.bisInVehicle {type=NiagaraBool}
 ]
}
