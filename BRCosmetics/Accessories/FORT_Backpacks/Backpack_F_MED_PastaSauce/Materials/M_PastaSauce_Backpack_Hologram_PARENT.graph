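# Parent translucent, unlit material for the hologram on the PastaSauce backpack.
# Pre-skinned position, passed through a vertex interpolator, drives a vertical dark-to-light purple gradient as the hologram base color.
# The same position's X and Z form planar texture coordinates with sign-based flipping, then parameterized centered scale and offset.
# The MF_HolographicTexture_PastaSauce function receives the color and coordinates to build the holographic look.
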Material /BRCosmetics/Accessories/FORT_Backpacks/Backpack_F_MED_PastaSauce/Materials/M_PastaSauce_Backpack_Hologram_PARENT
graph {
 properties {
  bUseMaterialAttributes=true
  bUsedWithClothing=true
  bUsedWithMeshDeformer=true
  bUsedWithNanite=true
  bUsedWithSkeletalMesh=true
  blendMode=BLEND_Translucent
  pixelDepthOffsetMode=PDOM_Legacy
  shadingModel=MSM_Unlit
  usedShadingModels=Unlit
 }
 nodes [
  n1 = VertexInterpolator(VS=PreSkinnedPosition())
  n2 = Multiply(A=Add(A=ComponentMask(None=@n1.PS) {b=true}) {constB=10}) {constB=0.06}
  n3 = Add(B=@n1.PS) {constA=1}
  n4 = Saturate(None=ComponentMask(None=Multiply(A=@n3) {constB=1000}) {b=true, g=true})
  n5 = Multiply(A=ComponentMask(None=@n1.PS) {b=true, r=true}, B=AppendVector(A=LinearInterpolate(Alpha=@n4) {constA=-1}, B=Constant() {r=-1}))
  n6 = ScalarParameter() {defaultValue=30, group=HolographicTexture, parameterName=Texture_UV_Scale}
  n7 = ScalarParameter() {group=HolographicTexture, parameterName=Texture_UV_Offset_X}
  n8 = ScalarParameter() {group=HolographicTexture, parameterName=Texture_UV_Offset_Y}
  MaterialFunctionCall!(
   MaterialAttributes=SetMaterialAttributes!(MaterialAttributes, Base Color=Constant()) {attributeSetTypes [#1]}
   HologramBaseColor=LinearInterpolate(
    A=Constant3Vector() {constant=<rgba 0.125 0.0312 0.0901 1>}
    B=Constant3Vector() {constant=<rgba 0.2961 0.0976 0.5457 1>}
    Alpha=Saturate(None=@n2)
   )
   HologramBaseColor_1
   Scanlines_UV
   TextureUV=Add(
    A=MaterialFunctionCall(UVs=@n5, Texture Scale=@n6) {materialFunction=&ScaleUVsByCenter}.UVs
    B=AppendVector(A=@n7, B=@n8)
   )
   HologramMask
  ) {materialFunction=/Game/Accessories/FORT_Backpacks/Backpack_F_MED_PastaSauce/Materials/MaterialFunctions/MF_HolographicTexture_PastaSauce}
 ]
}
